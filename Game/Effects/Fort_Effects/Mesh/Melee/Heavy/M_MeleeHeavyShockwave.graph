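# Masked, fully rough mesh-particle material for a heavy melee shockwave effect, with a warm gradient glow and noise-driven erosion.
# Inverted Fresnel plus a V-coordinate falloff blends Darker and Lighter colors, scaled by particle alpha and passed through the FX emissive adjustments function.
# Tiled noise plus the Erosion dynamic parameter is floored into a hard cutout, then multiplied by particle alpha.
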Material /Game/Effects/Fort_Effects/Mesh/Melee/Heavy/M_MeleeHeavyShockwave
graph {
 properties {bFullyRough=true, bUsedWithMeshParticles=true, blendMode=BLEND_Masked, pixelDepthOffsetMode=PDOM_Legacy, usedShadingModels=Default Lit}
 nodes [
  Constant()
  n1 = ParticleColor()
  n2 = VectorParameter() {parameterName=Darker, defaultValue=<rgba 0.515 0.1517 0.0258 0>}
  n3 = VectorParameter() {parameterName=Lighter, defaultValue=<rgba 3 2.1176 0.48 0>}
  n4 = ComponentMask(None=TextureCoordinate()) {g=true}
  n5 = Add(
   A=MaterialFunctionCall(Brightness=Constant() {r=1.5}, Exponent=Constant() {r=3}) {materialFunction=&MF_Fresnel_FX}.Fresnel Inverted
   B=OneMinus(None=Power(Base=@n4) {constExponent=1.5})
  )
  MaterialFunctionCall(Emissive=Multiply(A=@n1.A, B=LinearInterpolate(A=@n2.RGB, B=@n3.RGB, Alpha=@n5))) {materialFunction=&MF_FX_EmissiveAdjustments}
  n6 = TextureSample(UVs=TextureCoordinate() {uTiling=3, vTiling=0.5}) {texture=&T_NoisePack_01}
  n7 = DynamicParameter() {paramNames [Erosion,Param2,Param3,Param4]}
  Multiply(A=@n1.A, B=Floor(None=Add(A=@n6.R, B=@n7.Erosion)))
 ]
}
